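# Sound cue for the close sound of the EarthDay glider cosmetic in Athena.
# A random node picks one of four close wave variants with equal weight.
# It routes through the glider landing sound class, with shared open/close attenuation and concurrency and a cosmetic glider-close property sheet.
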SoundCue /BRCosmetics/Sounds/Gliders/EarthDay/Athena_Glider_EarthDay_Close_Cue
root=Random_0
nodes [
 {id=Random_0, children [WavePlayer_1,WavePlayer_2,WavePlayer_3,WavePlayer_4], properties {weights [1,1,1,1]}}
 {id=WavePlayer_1, properties {soundWaveAssetPtr=&/SFX_Glider_EarthDay_Close_01}}
 {id=WavePlayer_2, properties {soundWaveAssetPtr=&/SFX_Glider_EarthDay_Close_02}}
 {id=WavePlayer_3, properties {soundWaveAssetPtr=&/SFX_Glider_EarthDay_Close_03}}
 {id=WavePlayer_4, properties {soundWaveAssetPtr=&/SFX_Glider_EarthDay_Close_04}}
]
properties {
 attenuationSettings=/Game/Sounds/Attenuation/Fort_Player_Glider_OpenClose_Athena_Att
 audioPropertiesSheet=/BRCosmetics/Sounds/Settings/PropertySheets/Glider/APS_Cosmetic_Glider_Close
 maxDistance=10000
 soundClassObject=&Fort_Gliders_Land
 volumeMultiplier=1
 concurrencySet [&Athena_GliderOpenClose_Conc]
}
