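# Sound cue for the opening sound of the EarthDay cosmetic glider.
# A Random node with equal weights picks one of three wave players: Desert, ForestLake or Jungle open variants.
# It uses shared glider open/close attenuation and concurrency, a glider-open audio property sheet and the Fort_Gliders_Land sound class.
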SoundCue /BRCosmetics/Sounds/Gliders/EarthDay/Athena_Glider_EarthDay_Open_Cue
root=Random_0
nodes [
 {id=Random_0, children [WavePlayer_0,WavePlayer_2,WavePlayer_3], properties {weights [1,1,1]}}
 {id=WavePlayer_0, properties {soundWaveAssetPtr=&/SFX_Glider_EarthDay_Open_Desert}}
 {id=WavePlayer_2, properties {soundWaveAssetPtr=&/SFX_Glider_EarthDay_Open_ForestLake}}
 {id=WavePlayer_3, properties {soundWaveAssetPtr=&/SFX_Glider_EarthDay_Open_Jungle}}
]
properties {
 attenuationSettings=/Game/Sounds/Attenuation/Fort_Player_Glider_OpenClose_Athena_Att
 audioPropertiesSheet=/BRCosmetics/Sounds/Settings/PropertySheets/Glider/APS_Cosmetic_Glider_Open
 maxDistance=10000
 soundClassObject=&Fort_Gliders_Land
 volumeMultiplier=1
 concurrencySet [&Athena_GliderOpenClose_Conc]
}
